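; Weapon Blueprint for the Demon Hunter crossbow (Athena) that toggles the fire-tip flare effect as the weapon fires and reloads.
; Firing runs the parent fire FX, then deactivates the flare component.
; Reload start deactivates the flare if active, and reload end reactivates it only when the magazine holds ammo.
; The spawn-projectile event is empty and the construction script only calls its parent.
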(event EventOnPlayReloadFX
  (params (ReloadStage :type EFortReloadFXState%20Enum :default "ReloadStart"))
  (bind _p_demonhunter_crossbow_firetipflare (Variables|Default|GetP_DemonHunter_Crossbow_FireTipFlare))
  (switch Utilities|FlowControl|Switch|SwitchonEFortReloadFXState ReloadStage
    (:ReloadStart
      (if (Components|Activation|IsActive _p_demonhunter_crossbow_firetipflare)
        (Components|Activation|Deactivate _p_demonhunter_crossbow_firetipflare)))
    (:ReloadEnd
      (if (> (FortWeapon|GetMagazineAmmoCount) 0)
        (Components|Activation|Activate (Variables|Default|GetP_DemonHunter_Crossbow_FireTipFlare) false)))))

(event FortWeapon|EventOnSpawnProjectile
  (params (SpawnProjectile :type Fort%20Projectile%20Base%20Object%20Reference)))

(event EventOnPlayWeaponFireFX
  (params (bPersistentFire :type Boolean) (bSecondaryFire :type Boolean))
  ; New way to deactivate fx
  (|CallParentFunction bPersistentFire bSecondaryFire)
  (Components|Activation|Deactivate (Variables|Default|GetP_DemonHunter_Crossbow_FireTipFlare)))

(fn ConstructionScript
  (|CallParentFunction))
