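; Rope actor for the hook gun that plays the visual, audio and gameplay-event side of reeling the hook back in.
; On End Reel, a timeline moves the projectile mesh from end to start, stretches the rope by distance and drives material wiggle.
; On finish it sends the reattach event, plays audio from authority only because remote firing is unreliable, then destroys itself.
; Destruction or pawn ability-system invalidation removes rope-active and pickup-blocking tags.
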(event EventBeginPlay
  (bind _playerpawn (Variables|Default|GetPlayerPawn))
  ;[0.5448 1 0.8118] EVENT -> BeginPlay
  (EventDispatchers|BindEventtoOnEndReel (EventDispatchers|CreateEvent :!event Custom|End%20Reel))
  (Utilities|IsValid _playerpawn
    (:Is%20Valid
      (EventDispatchers|BindEventtoOnPreInvalidateAbilitySystemComponent _playerpawn (EventDispatchers|CreateEvent :!event Custom|OnPlayerPawnPreInvalidateAbilitySystemComponent)))))

(event Custom|TightenRope
  ;[1 1 1] Tighten Rope
  (bind (_direction _newtrack_0) (Timeline|Tight PlayFromStart)
    (:Update
      (Rendering|Material|SetScalarParameterValueonMaterials (Variables|StaticMesh|GetRopeMesh) "WiggleAmount" (Math|Float|Lerp 0.0 (Variables|Default|GetMaxWiggleAmount) _newtrack_0)))))

(event Custom|PlayReattachAudio
  ;[1 1 1] Reel In
  (Networking|CanExecuteCosmeticEvents
    (:True
      (bind _rootcomponent (Variables|/Script/Engine.Actor|GetRootComponent (Variables|Default|GetPlayerPawn))
        (:then
          (Audio|SpawnSoundAttached (Variables|RopeSounds|GetReattachSound) _rootcomponent))
        (:else)))))

(event Custom|End%20Reel
  (bind _ropeprojectilemesh (Variables|StaticMesh|GetRopeProjectileMesh))
  (bind _ropemesh (Variables|StaticMesh|GetRopeMesh))
  (bind _startlocation (Variables|FortHookGunRope|GetStartLocation))
  (bind _endlocation (Variables|FortHookGunRope|GetEndLocation))
  ;[1 1 1] Reel In
  (Utilities|FlowControl|Sequence
    (:then_0
      (bind _rootcomponent (Variables|/Script/Engine.Actor|GetRootComponent (Variables|Default|GetPlayerPawn))
        (:then
          (Audio|SpawnSoundAttached :Sound (Variables|RopeSounds|GetReelSound) :AttachToComponent _rootcomponent :AttachPointName "None" :LocationType "KeepRelativeOffset" :bStopWhenAttachedToDestroyed true :VolumeMultiplier 2.0 :PitchMultiplier 1.5 :StartTime 0.0 :bAutoDestroy true))
        (:else
          _))
      (Rendering|SetVisibility (Variables|StaticMesh|GetRopeProjectileMesh) true false))
    (:then_1
      (Ability|SendGameplayEventtoActor (Variables|Default|GetPlayerPawn) (Variables|Default|GetEvent_StartReelIn))
      (bind (_direction _reel_lerp_controller _tighten) (Timeline|Reel%20In PlayFromStart)
        (:Update
          (CallFunction|SetStartPointOfRope)
          (Transformation|SetWorldLocation _ropeprojectilemesh (Math|Vector|Lerp(Vector) _endlocation _startlocation _reel_lerp_controller) false false)
          (Transformation|SetWorldRotation _ropeprojectilemesh (Math|Rotator|FindLookatRotation _startlocation _endlocation) false false)
          (bind _distance (Class|MissionBlueprintFunctionLibrary|DistanceBetweenTwoVectors _startlocation (Math|Vector|Lerp(Vector) _endlocation _startlocation _reel_lerp_controller)))
          (Transformation|SetRelativeScale3D (Variables|StaticMesh|GetRopeMesh) (Math|Vector|MakeVector (/ _distance 1e+02) 1.0 1.0))
          (Rendering|Material|SetScalarParameterValueonMaterials _ropemesh "ReturnWiggle" (Math|Float|Lerp 0.0 3.0 _tighten))
          (Rendering|Material|SetScalarParameterValueonMaterials _ropemesh "WiggleAmount" (Math|Float|Lerp 0.0 (Variables|Default|GetMaxWiggleAmount) _tighten))
          (Transformation|SetWorldRotation (Variables|StaticMesh|GetRopeMesh) (Math|Rotator|FindLookatRotation _startlocation _endlocation) false false)
          (Utilities|IsValid (Actor|GetOwner)
            (:Is%20Not%20Valid
              (Rendering|SetVisibility (Variables|StaticMesh|GetRopeProjectileMesh) false false)
              (Actor|DestroyActor))))
        (:Finished
          (Ability|SendGameplayEventtoActor (Variables|Default|GetPlayerPawn) (Variables|Default|GetEvent_Reattach)) ; Plays reattach anim
          (Networking|SwitchHasAuthority
            (:Authority
              (CallFunction|PlayReattachAudio))
            (:Remote
              _)) ; Most of the time Server and Client fire through here.\nSometimes, only the server does. Then the reattach sound doesn't play.\nSo now, only play the sound on server, replicated to clients.
          (Rendering|SetVisibility (Variables|StaticMesh|GetRopeProjectileMesh) false false)
          (Actor|DestroyActor))))))

(event EventDestroyed
  ;[1 0.6595 0.4235] Event Destroyed and OnPrePlayerPawnDetachAbilitySystemComponent - Remove Rope Active and Block Incoming Pickups tags
  (Utilities|IsValid (Variables|Default|GetPlayerPawn)
    (:Is%20Valid
      (Tags|RemoveNonReplicatedTagtoPawn (Variables|Default|GetPlayerPawn) (Variables|Default|GetRopeActiveTag))
      (Tags|RemoveNonReplicatedTagtoPawn (Variables|Default|GetPlayerPawn) (Variables|Default|GetBlockIncomingPickups))
      (EventDispatchers|UnbindEventfromOnPreInvalidateAbilitySystemComponent (Variables|Default|GetPlayerPawn) (EventDispatchers|CreateEvent :!event Custom|OnPlayerPawnPreInvalidateAbilitySystemComponent)))))

(event Custom|OnPlayerPawnPreInvalidateAbilitySystemComponent
  ;[1 0.6595 0.4235] Event Destroyed and OnPrePlayerPawnDetachAbilitySystemComponent - Remove Rope Active and Block Incoming Pickups tags
  (Utilities|IsValid (Variables|Default|GetPlayerPawn)
    (:Is%20Valid
      (Tags|RemoveNonReplicatedTagtoPawn (Variables|Default|GetPlayerPawn) (Variables|Default|GetRopeActiveTag))
      (Tags|RemoveNonReplicatedTagtoPawn (Variables|Default|GetPlayerPawn) (Variables|Default|GetBlockIncomingPickups))
      (EventDispatchers|UnbindEventfromOnPreInvalidateAbilitySystemComponent (Variables|Default|GetPlayerPawn) (EventDispatchers|CreateEvent :!event Custom|OnPlayerPawnPreInvalidateAbilitySystemComponent)))))

(fn ConstructionScript)
