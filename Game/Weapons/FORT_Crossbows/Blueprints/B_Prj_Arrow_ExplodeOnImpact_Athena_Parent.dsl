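; Parent Blueprint for the explode-on-impact Boom Bow arrow projectile, handling its tracer, impact replication and explosion effects.
; BeginPlay animates a fake tracer through material parameters and arms a fallback explosion timer for long bouncy flights.
; On blocking touches the server sets hit data, and OnRep_HitLocation plays explosion FX, sound, indicator, feedback cue and decal.
; Overlaps with a valid physical material play water impact effects and destroy the arrow.
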(event EventBeginPlay
  (bind _tracermesh (Variables|Default|GetTracerMesh))
  ;[1 1 1] EVENT -> BeginPlay
  ;;[1 1 1] With bouncy surfaces, there's a possibility for a long life. Explode after a time
  (Utilities|Time|SetTimerbyEvent (EventDispatchers|CreateEvent :!event DoExplosion) 22.0)
  ;[1 1 1] Animate fake tracer mesh
  (Networking|CanExecuteCosmeticEvents
    (:True
      (bind _asfort_player_pawn (Utilities|Casting|CastToFortPlayerPawn (Variables|Actor|GetInstigator)))
      (Rendering|Material|SetVectorParameterValueonMaterials _tracermesh "BulletOrigin" (Effects|GetMuzzleLocation (Variables|/Script/FortniteGame.FortPawn|GetCurrentWeapon _asfort_player_pawn) -1))
      (Rendering|Material|SetScalarParameterValueonMaterials _tracermesh "StartTime" (Utilities|Time|GetGameTimeinSeconds))
      (Rendering|Material|SetScalarParameterValueonMaterials _tracermesh "WPODuration" (Math|Float|MapRangeClamped (Math|Vector|VectorLength (Transformation|GetVelocity)) 5e+03 4e+04 0.1 0.25)))))

(event EventOnTouched
  (params (OtherActor :type Actor%20Object%20Reference) (OtherComp :type Primitive%20Component%20Object%20Reference) (HitResult :type Hit%20Result%20Structure%20%28by%20ref%29) (bIsOverlap :type Boolean))
  (bind (_bblockinghit _binitialoverlap _time _distance _location _impactpoint _normal _impactnormal _physmat _hitactor) (Collision|BreakHitResult HitResult))
  (bind _waterinteractioncomponent (Variables|Water|GetWaterInteractionComponent))
  (bind (_waterplanelocation _waterplanenormal _watersurfaceposition _waterdepth _waterbodyidx) (FX|GetLastWaterSurfaceInfo _waterinteractioncomponent :!class /Script/FortniteGame.FortWaterInteractionComponent))
  (bind (_asfort_pawn _bsuccess) (Utilities|Casting|CastToFortPawn _hitactor))
  ;[1 1 1] EVENT -> OnTouched
  (if (Utilities|Array|ContainsItem (Variables|Default|GetClassesToIgnoreTouch) (Utilities|GetClass OtherActor))
    _
    (elif bIsOverlap
      ;;[1 0 0] Break Hit Result didn't return surface type. So changed it.
      (Utilities|IsValid _physmat
        (:Is%20Valid
          (Audio|PlaySoundatLocation "/Game/Sounds/Fort_Weapons/Explosive_Bow/Weapon_Compound_Bow_Proj_Explo.Weapon_Compound_Bow_Proj_Explo" _location)
          (bind _returnvalue (Effects|Components|ParticleSystem|SpawnFXAtLocation (Variables|Default|GetWaterImpactFX_NS) _location (Math|Conversions|RotationFromXVector _waterplanenormal)))
          (Effects|Components|ParticleSystem|SetFloatParameter _returnvalue "WaterDepth" _waterdepth)
          (Effects|Components|ParticleSystem|SetFloatParameter _returnvalue "Global Scale" 0.5)
          (Effects|Components|ParticleSystem|SetFloatParameter _returnvalue "WaterBodyIndex" (Math|Conversions|ToFloat(Integer) _waterbodyidx))
          (Actor|DestroyActor)))
      (else
        (Networking|SwitchHasAuthority
          (:Authority
            (if (not (Delivery|IsBeingKilled))
              (Variables|BPrjBoomBow|SetHitLocation _location :!pinned true) ; All Explosion FX
              (Variables|BPrjBoomBow|SetHitRotation (Math|Rotator|MakeRotfromX _normal))
              (Variables|BPrjBoomBow|SetHitPlayer _bsuccess))))))))

(event OnEnterWaterEvent(WaterInteractionComponent)
  (params (WaterBody :type Fort%20Water%20Body%20Actor%20Object%20Reference) (WaterInteractionComponent :type Fort%20Water%20Interaction%20Component%20Object%20Reference) (bIsFirstBody :type Boolean))
  (Audio|PlaySoundatLocation "/Game/Athena/Sounds/Weapons/Explosions/Expl_Debris_Water_Cue.Expl_Debris_Water_Cue" (Transformation|GetActorLocation)))

(fn OnRep_HitLocation
  (bind _hitlocation (Variables|BPrjBoomBow|GetHitLocation))
  (bind _hitrotation (Variables|BPrjBoomBow|GetHitRotation))
  (bind _hitlocation_2 (Variables|BPrjBoomBow|GetHitLocation))
  (if (Variables|BPrjBoomBow|GetUse%20Niagara%20Explosion)
    (bind _returnvalue (Effects|Components|ParticleSystem|SpawnFXAtLocation (Variables|BPrjBoomBow|GetNS_Explosion) _hitlocation _hitrotation))
    (Effects|Components|ParticleSystem|SetFloatParameter _returnvalue "Global Scale" 0.45)
    (else
      (bind _returnvalue_2 (Effects|SpawnEmitteratLocation (Variables|BPrjBoomBow|GetP_Explosion) _hitlocation _hitrotation))))
  (Rendering|SetRenderCustomDepth _returnvalue_2 true)
  (LOD|SetMaxDrawDistance _returnvalue_2 (Variables|Default|GetDrawDistance))
  (Components|Activation|Deactivate (Variables|BPrjBoomBow|GetProjectileTrail))
  (Utilities|FlowControl|Sequence
    (:then_0
      (GameplayCue|ExecuteGameplayCue%28Burst%29Local self (Variables|Default|GetFeedbackCue) (Ability|GameplayCue|MakeGameplayCueParameters 0.0)) ; Cam shake + controller vibration
      (if (not (or (Fortnite|IsMobilePlatform) (Variables|BPrjBoomBow|GetHitPlayer)))
        (bind _returnvalue_3 (Rendering|Decal|SpawnDecalatLocation "/Game/Weapons/Materials/MD_Decal_BulletHole_Lifetime_Long_Inst.MD_Decal_BulletHole_Lifetime_Long_Inst" (Variables|Default|GetDecalSize) (Variables|BPrjBoomBow|GetHitLocation) (Math|Rotator|InvertRotator (Variables|BPrjBoomBow|GetHitRotation)) 4.5))
        (Rendering|Components|Decal|SetFadeOut _returnvalue_3 0.1 5.0 true)))
    (:then_1
      (Audio|PlaySoundatLocation (Variables|Sounds|GetCue_CloseSound) _hitlocation_2) ; Play Close Explosion
      (Fortnite|ShowSoundIndicatorLocalforTeam :Location _hitlocation_2 :MaxAudibleDistance (Variables|/Script/Engine.SoundBase|GetMaxDistance (Variables|Sounds|GetCue_DistantSound)) :IndicatorType "Gunshot" :Team (Team|GetActorTeam self) :Affiliations (Utilities|Array|MakeArray "Hostile" "Neutral") :Instigator self :Tint "(R=0.000000,G=0.000000,B=0.000000,A=0.000000)" :AttenuationMaxRangeMultiplier 1.0 :AttenuationIntensityMultiplier 1.0) ; Visual Sound Indicator
    )))

(fn ConstructionScript
  (LOD|SetMaxDrawDistance (Variables|Default|GetTracerMesh) (Variables|Default|GetDrawDistance)))
