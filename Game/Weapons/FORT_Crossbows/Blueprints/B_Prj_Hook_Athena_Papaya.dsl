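; Projectile Blueprint for the Papaya hook gun that creates the rope linking the weapon to the fired hook.
; SpawnRope spawns a B_HookGunRope_Papaya actor at the projectile's location, passing in the weapon actor and player pawn, then stores it as Rope.
; It hands the weapon and itself to the rope and returns it.
; The BeginPlay, overlap and tick events have no active logic.
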(event EventBeginPlay
  (|CallParentFunction) ; This node is disabled and will not be called.\nDrag off pins to build functionality.
)

(event Collision|EventActorBeginOverlap
  (params (OtherActor :type Actor%20Object%20Reference)))

(event EventTick
  (params (DeltaSeconds :type Float%20%28single-precision%29)))

(fn SpawnRope
  (params (PlayerPawn :type Fort%20Player%20Pawn%20Object%20Reference))
  ;[1 1 1] Spawn The Rope
  (bind _returnvalue (Game|SpawnActorfromClass "/Game/Weapons/FORT_Crossbows/Blueprints/B_HookGunRope_Papaya.B_HookGunRope_Papaya_C" (Math|Transform|MakeTransform (Transformation|GetActorLocation)) "Undefined" "MultiplyWithRoot" (Variables|Default|GetWeapon%20Actor) PlayerPawn 0))
  (bind _output_get (Variables|Default|SetRope _returnvalue))
  (Class|FortHookGunRope|SetWeaponandProjectile _output_get (Variables|Default|GetWeapon%20Actor) self)
  (return _output_get))

(fn ConstructionScript
  (|CallParentFunction))
