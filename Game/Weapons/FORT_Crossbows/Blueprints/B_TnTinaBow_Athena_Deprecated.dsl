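; Deprecated Athena charge bow weapon (TnTinaBow) that shows a predicted arrow trajectory while the local player draws it.
; Charge start spawns the trajectory actor, and Tick recalculates velocity from charge percent to rebuild its one-bounce spline from dummy projectile data.
; Max charge adds sound, Niagara, camera shake, rumble and a trajectory pulse.
; Firing picks tension-based audio and shows a sound indicator; firing, release or unequip tears down the preview.
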(event Custom|IncreaseDrawStrength
  (bind _maxprojectilespeed (Variables|ExplosiveBow|GetMaxProjectileSpeed))
  (bind _returnvalue (ScalableFloat|GetValueatLevel _maxprojectilespeed 1.0))
  ;[1 1 1] Update Trajectory Spline
  (if (<= (+ (Variables|ExplosiveBow|GetProjectileVelocity) 1.0) _returnvalue)
    (Variables|ExplosiveBow|SetProjectileVelocity (CallFunction|CalculateVelocity))
    (else
      (Variables|ExplosiveBow|SetProjectileVelocity _returnvalue)))
  (CallFunction|CalculateTrajectorySpline)
  (Variables|ExplosiveBow|SetReachedHalfCharge (>= (FortWeapon|GetChargePercent) 0.5)))

(event EventOnStartCharge
  ;[1 1 1] EVENT -> StartCharge
  (|CallParentFunction)
  (bind _asfort_player_pawn (Utilities|Casting|CastToFortPlayerPawn (Game|GetInstigator)))
  (bind _output_get (Variables|ExplosiveBow|SetPlayerPawn _asfort_player_pawn))
  (CallFunction|StartDrawAudio)
  (if (Pawn|IsLocallyControlled _output_get)
    (CallFunction|SetupTrajectoryProjectile)
    (CallFunction|SetDummyVariables)))

(event EventOnChargeUp
  ;[1 1 1] EVENT -> On Charge Up
  (|CallParentFunction)
  (Utilities|FlowControl|DoOnce false
    (:Completed
      (Variables|ExplosiveBow|SetCharging true)
      (Variables|ExplosiveBow|SetReached_Min_Charge false) ; Can play charged weapon fire sound
      (Variables|ExplosiveBow|SetReached_Max_Charge false))))

(event Custom|ResetCharge
  ;[1 1 1] EVENT -> On Charge Up
  (Variables|ExplosiveBow|SetProjectileVelocity 0.0)
  (Utilities|FlowControl|DoOnce false
    (:Completed
      (Variables|ExplosiveBow|SetCharging true)
      (Variables|ExplosiveBow|SetReached_Min_Charge false) ; Can play charged weapon fire sound
      (Variables|ExplosiveBow|SetReached_Max_Charge false))))

(event EventOnLocalClientReachedMaxCharge
  (bind _returnvalue (Game|GetInstigator))
  (bind _returnvalue_2 (Fortnite|GetFortPlayerControllerfromActor _returnvalue))
  ;[1 1 1] EVENT -> Min/Max Charge Reached
  (Variables|ExplosiveBow|SetReached_Max_Charge true)
  (Audio|Components|Audio|Play (Variables|Default|GetFullCharge_AudioComponent) 0.0)
  (Niagara|SpawnSystemAttached (Variables|Default|GetFullyChargeFX_NS) (Variables|Visual|GetWeaponMesh) "Muzzle")
  (Game|Feedback|ClientStartCameraShake :self _returnvalue_2 :Shake "/Game/Effects/Fort_Effects/Camera_Effects/BP_Camera_Shake_Pulse_Zoom_Small.BP_Camera_Shake_Pulse_Zoom_Small_C" :Scale 1.0 :PlaySpace "CameraLocal")
  (Game|Feedback|ClientPlayForceFeedback _returnvalue_2 "/Game/Weapons/ForceFeedback/FF_ShoppingCartBump.FF_ShoppingCartBump" "None" false true false)
  (bind _bowtrajectory (Variables|ExplosiveBow|GetBowTrajectory)
    (:then
      (Class|BPProjectileTrajectoryBowGeneric|Pulse _bowtrajectory))
    (:else)))

(event EventOnLocalClientReachedMinCharge
  ;[1 1 1] EVENT -> Min/Max Charge Reached
  (Variables|ExplosiveBow|SetReached_Min_Charge true) ; Can play charged weapon fire sound
)

(event Custom|EndOfCharge
  ;[1 1 1] End of Charge (Max Charge Reached)
  (CallFunction|StopDrawAudio)
  (Variables|ExplosiveBow|SetProjectileVelocity 0.0)
  (Utilities|Time|ClearandInvalidateTimerbyHandle (Variables|ExplosiveBow|GetUpdateTrajectorySpline))
  (Actor|DestroyActor (Variables|ExplosiveBow|GetBowTrajectory)))

(event EventOnChargeDown
  ;[1 1 1] EVENT -> Charge Down (mouse released)
  (CallFunction|EndOfCharge)
  (Utilities|FlowControl|Delay 0.034)
  (Variables|ExplosiveBow|SetCharging false))

(event EventOnPlayWeaponFireFX
  (params (bPersistentFire :type Boolean) (bSecondaryFire :type Boolean))
  (bind _returnvalue (Game|GetInstigator))
  (bind _reached_max_charge (Variables|ExplosiveBow|GetReached_Max_Charge))
  (bind _fulltensionfire (Variables|FireTensionSounds|GetFullTensionFire))
  (bind _reachedhalfcharge (Variables|ExplosiveBow|GetReachedHalfCharge))
  (bind _normaltensionfire (Variables|FireTensionSounds|GetNormalTensionFire))
  (bind _lowtensionfire (Variables|FireTensionSounds|GetLowTensionFire))
  (bind _returnvalue_2 (select _reachedhalfcharge _normaltensionfire _lowtensionfire))
  (bind _returnvalue_3 (select _reached_max_charge _fulltensionfire _returnvalue_2))
  (bind _rootcomponent (Variables|/Script/Engine.Actor|GetRootComponent _returnvalue))
  ;[1 1 1] EVENT -> Weapon Fire FX
  (|CallParentFunction bPersistentFire false)
  (Utilities|FlowControl|Sequence
    (:then_0
      (CallFunction|EndOfCharge)
      (if (Pawn|IsLocallyControlled (Game|GetInstigator))
        (Audio|SpawnSoundAttached (select _reached_max_charge (Variables|FireTensionSounds|GetFullTensionFireP1) (select _reachedhalfcharge (Variables|FireTensionSounds|GetNormalTensionFireP1) (Variables|FireTensionSounds|GetLowTensionFireP1))) _rootcomponent)
        (else
          (Audio|SpawnSoundAttached _returnvalue_3 _rootcomponent)))
      (Fortnite|ShowSoundIndicatorLocalforTeam :Location (Transformation|GetActorLocation self) :MaxAudibleDistance (Variables|/Script/Engine.SoundBase|GetMaxDistance _returnvalue_3) :IndicatorType "Gunshot" :Team (Team|GetActorTeam self) :Affiliations (Utilities|Array|MakeArray "Hostile" "Neutral") :Instigator self :Tint "(R=0.000000,G=0.000000,B=0.000000,A=0.000000)" :AttenuationMaxRangeMultiplier 1.0 :AttenuationIntensityMultiplier 1.0) ; Visual Sound Indicator. Need this because we don't have a fire sound set in Class Defaults
    )
    (:then_1
      (if (Variables|ExplosiveBow|GetReached_Max_Charge)
        (Niagara|SpawnSystemAttached (Variables|Default|GetAdditionalFireFX_NS) (Variables|Visual|GetWeaponMesh) "Muzzle")))))

(event FortWeapon|EventOnEquippedAndAttached
  ;[1 1 1] EVENT -> Equip: Added 'IsEquipped' bool
  (Variables|ExplosiveBow|SetIsBowEquipped true))

(event FortWeapon|EventOnLocalUnEquip
  ;[1 1 1] EVENT -> Unequip - to fix a bug if you were charging, then switched weapons before firing
  (Variables|ExplosiveBow|SetIsBowEquipped false)
  (CallFunction|ResetCharge)
  (Variables|ExplosiveBow|SetCharging false)
  (CallFunction|EndOfCharge))

(event EventOnEndCharge
  ;[1 1 1] EVENT -> Unequip - to fix a bug if you were charging, then switched weapons before firing
  (CallFunction|ResetCharge)
  (Variables|ExplosiveBow|SetCharging false)
  (CallFunction|EndOfCharge))

(event EventTick
  (params (DeltaSeconds :type Float%20%28single-precision%29))
  (bind _returnvalue (Game|GetInstigatorController))
  ;[1 1 1] Tick - Update trajectory for local instigator
  (|CallParentFunction DeltaSeconds)
  (Utilities|IsValid (Variables|ExplosiveBow|GetBowTrajectory)
    (:Is%20Valid
      (Utilities|IsValid _returnvalue
        (:Is%20Valid
          (if (and (Pawn|IsLocalPlayerController _returnvalue) (> (FortWeapon|GetChargePercent) 0.0))
            (CallFunction|IncreaseDrawStrength)))))))

(fn CalculateTrajectorySpline
  (bind (_startlocation _initialvelocity _projectilevelocity) (CallFunction|SetupProjectileVariables))
  (FortWeapon|UpdateDamageStartLocation)
  (bind (_outhitresult _outsplinepoints _outsplinetangents) (Projectile|CalculateProjectileTrajectorySplinewithHit :Instigator (Variables|ExplosiveBow|GetPlayerPawn) :InitialLocation _startlocation :InitialVelocity _initialvelocity :MaxSpeed _projectilevelocity :Gravity (Variables|ExplosiveBow|GetDummyGravity) :Friction (Variables|ExplosiveBow|GetDummyFriction) :Bounciness (Variables|ExplosiveBow|GetDummyBouncieness) :TimeStep (Variables|ExplosiveBow|GetDummyTimestep) :TraceExtent (Variables|ExplosiveBow|GetDummyExtent) :TraceChannel (Variables|ExplosiveBow|GetDummyCollisionChannel) :MaxBounces 1 :MaxSteps 20 :MaxDistanceBetweenSplinePoints (Math|Float|Lerp 3.5e+02 1e+03 (FortWeapon|GetChargePercent)) :InitialDistance 1.5e+02 :LinearDamping 0.0 :bReturnPhysicalMaterial false))
  (Trajectory|SetTrajectorySpline (Variables|ExplosiveBow|GetBowTrajectory) _outsplinepoints _outsplinetangents)
  (return))

(fn CalculateVelocity
  (return (FortWeapon|GetProjectileSpeed (FortWeapon|GetChargePercent))))

(fn SetDummyVariables
  (bind _trajectorydata (Class|FortKismetLibrary|GetProjectileTrajectoryData (Variables|ExplosiveBow|GetPlayerPawn) (Variables|ExplosiveBow|GetPrj_Dummy)))
  (Variables|ExplosiveBow|SetDummyMaxSpeed (Utilities|Struct|BreakFortProjectileTrajectoryData _trajectorydata))
  (Variables|ExplosiveBow|SetDummyGravity (.GravityZ (Utilities|Struct|BreakFortProjectileTrajectoryData _trajectorydata)))
  (Variables|ExplosiveBow|SetDummyExtent (Math|Float|Max(Float) (.CapsuleRadius (Utilities|Struct|BreakFortProjectileTrajectoryData _trajectorydata)) (.CapsuleHalfHeight (Utilities|Struct|BreakFortProjectileTrajectoryData _trajectorydata))))
  (Variables|ExplosiveBow|SetDummyBouncieness (.Bounciness (Utilities|Struct|BreakFortProjectileTrajectoryData _trajectorydata)))
  (Variables|ExplosiveBow|SetDummyFriction (.Friction (Utilities|Struct|BreakFortProjectileTrajectoryData _trajectorydata))))

(fn SetupProjectileVariables
  (bind (_aimdir _returnvalue) (FortWeapon|GetDamageStartLocation -1))
  (return _returnvalue (* (Math|Rotator|GetRotationXVector (Math|Rotator|MakeRotator (.roll (Math|Conversions|RotationFromXVector _aimdir)) (+ (.pitch (Math|Conversions|RotationFromXVector _aimdir)) (FortWeapon|GetProjectilePitchOffset)) (.yaw (Math|Conversions|RotationFromXVector _aimdir)))) (Variables|ExplosiveBow|GetProjectileVelocity)) (Variables|ExplosiveBow|GetProjectileVelocity) (Math|Rotator|MakeRotator (.roll (Math|Conversions|RotationFromXVector _aimdir)) (+ (.pitch (Math|Conversions|RotationFromXVector _aimdir)) (FortWeapon|GetProjectilePitchOffset)) (.yaw (Math|Conversions|RotationFromXVector _aimdir)))))

(fn SetupTrajectoryProjectile
  (bind _returnvalue (Game|SpawnActorfromClass (Variables|ExplosiveBow|GetTrajectoryClass) (Transformation|GetActorTransform (Variables|ExplosiveBow|GetPlayerPawn)) "AlwaysSpawn" "SelectDefaultAtRuntime" (Variables|ExplosiveBow|GetPlayerPawn) (Variables|ExplosiveBow|GetPlayerPawn)))
  (bind _asbp_projectile_trajectory_explosive_bow (Utilities|Casting|CastToBP_ProjectileTrajectory_ExplosiveBow _returnvalue))
  (Variables|/Game/Athena/Items/Weapons/Abilities/ExplosiveBow/BP_ProjectileTrajectory_ExplosiveBow.BP_ProjectileTrajectory_ExplosiveBow_C|SetOwnerWeapon self _asbp_projectile_trajectory_explosive_bow)
  (bind _output_get (Variables|ExplosiveBow|SetBowTrajectory _asbp_projectile_trajectory_explosive_bow))
  (Actor|Tick|AddTickPrerequisiteActor _output_get self))

(fn StartDrawAudio
  (Variables|Default|GetChargeAudio
    (:then)
    (:else
      (bind _returnvalue (Audio|SpawnSoundAttached (Variables|Default|GetSoundOnCharge) (Variables|Visual|GetWeaponMesh)))
      (Variables|Default|SetChargeAudio _returnvalue))))

(fn StopDrawAudio
  (bind _chargeaudio (Variables|Default|GetChargeAudio)
    (:then
      (Audio|Components|Audio|FadeOut _chargeaudio 0.15 0.0 "Linear" :!class /Script/Engine.AudioComponent))
    (:else)))

(fn ConstructionScript
  (|CallParentFunction))
